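# Animation Blueprint for the M_MED_Donut player body that adds secondary motion to holster, pack and glute bones.
# It copies the attached parent's pose, then weights holster and pack parents between body bones and simulates packs with spring AnimDynamics scaled by DynamicsAlpha.
# Additive bone offsets follow thigh alphas for glutes and CrouchAlpha, toggled by crouch events, for packs and left holster.
# Two leg-pack AnimDynamics nodes sit unconnected to the output.
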AnimBlueprint /Game/Characters/Player/Male/Medium/Bodies/M_MED_Donut/Meshes/M_MED_Donut_AnimBP
target_skeleton=&/M_MED_Donut_Skeleton
graphs [
 AnimGraph {nodes [
   n1 = Anim Dynamics - Bone: dyn_legpackA_01_r(
    ComponentPose=Reroute Node(InputPin:Component Space Pose Link Structure) -> (OutputPin:Component Space Pose Link Structure)
    Alpha=0.37
    ExternalForce="0,0,0"
   ) {
    angularDampingOverride=1
    angularSpringConstant=1500
    bAngularSpring=true
    bGravityOverrideInSimSpace=true
    bOverrideAngularDamping=true
    bUseGravityOverride=true
    gravityScale=0.35
    simulationSpace=RootRelative
    alphaBoolBlend {blendOption=Linear}
    boundBone {boneName=dyn_legpackA_01_r}
    componentAppliedLinearAccClamp=<3000 3000 3500>
    componentLinearAccScale=<0.3 0.3 0.3>
    componentLinearVelScale=<0.3 0.3 1.5>
    gravityOverride {z=-980}
    physicsBodyDefinitions [{
      sphereCollisionRadius=0
      boundBone {boneName=dyn_legpackA_01_r}
      boxExtents=<20 20 10>
      constraintSetup {angularLimitsMax=<5 0 5>, angularLimitsMin=<-5 -45 -5>, angularTarget=<0 0 0>}
      localJointOffset=<4 0 0>
     }]
   }
   Anim Dynamics - Bone: dyn_legpackB_01_r(ComponentPose=@n1, Alpha=0.5, ExternalForce="0,0,0") {
    angularDampingOverride=1
    angularSpringConstant=1500
    bAngularSpring=true
    bGravityOverrideInSimSpace=true
    bOverrideAngularDamping=true
    bUseGravityOverride=true
    gravityScale=0.35
    simulationSpace=RootRelative
    alphaBoolBlend {blendOption=Linear}
    boundBone {boneName=dyn_legpackB_01_r}
    componentAppliedLinearAccClamp=<3000 3000 3500>
    componentLinearAccScale=<0.3 0.3 0.3>
    componentLinearVelScale=<0.3 0.3 1.5>
    gravityOverride {z=-980}
    physicsBodyDefinitions [{
      sphereCollisionRadius=0
      boundBone {boneName=dyn_legpackB_01_r}
      boxExtents=<20 20 10>
      constraintSetup {angularLimitsMax=<5 0 5>, angularLimitsMin=<-5 -45 -5>, angularTarget=<0 0 0>}
      localJointOffset=<4 0 0>
     }]
   }
   n2 = Get DynamicsAlpha() -> (DynamicsAlpha:float)
   n3 = "Constraint - dyn_gun_holster_l (PPP)"!(
    ComponentPose=Local To Component(LocalPose=Copy Pose From Mesh() {bUseAttachedParent=true})
    Alpha=1
    ConstraintWeights_0=0.2545
    ConstraintWeights_1:float=0.7242
    ConstraintWeights_2:float=0.0213
   ) -> (Pose) {
    alphaBoolBlend {blendOption=Linear}
    boneToModify {boneName=dyn_gun_holster_l}
    constraintSetup [{targetBone {boneName=pelvis}}, {targetBone {boneName=thigh_twist_01_l}}, {targetBone {boneName=thigh_l}}]
    constraintWeights [1,1,1]
   }
   n4 = "Constraint - dyn_gun_holster_r (PPP)"!(ComponentPose=@n3, Alpha=1, ConstraintWeights_0=0.0182, ConstraintWeights_1:float=0.2193, ConstraintWeights_2:float=0.7625) -> (Pose) {
    alphaBoolBlend {blendOption=Linear}
    boneToModify {boneName=dyn_gun_holster_r}
    constraintSetup [{targetBone {boneName=thigh_r}}, {targetBone {boneName=pelvis}}, {targetBone {boneName=thigh_twist_01_r}}]
    constraintWeights [1,1,1]
   }
   n5 = "Constraint - dyn_parent_packA_l (PP)"!(ComponentPose=@n4, Alpha=1, ConstraintWeights_0=0.6989, ConstraintWeights_1:float=0.3011) -> (Pose) {
    alphaBoolBlend {blendOption=Linear}
    boneToModify {boneName=dyn_parent_packA_l}
    constraintSetup [{targetBone {boneName=pelvis}}, {targetBone {boneName=spine_01}}]
    constraintWeights [1,1]
   }
   n6 = "Constraint - dyn_parent_packB_l (PP)"!(ComponentPose=@n5, Alpha=1, ConstraintWeights_0=0.699, ConstraintWeights_1:float=0.301) -> (Pose) {
    alphaBoolBlend {blendOption=Linear}
    boneToModify {boneName=dyn_parent_packB_l}
    constraintSetup [{targetBone {boneName=pelvis}}, {targetBone {boneName=spine_01}}]
    constraintWeights [1,1]
   }
   n7 = "Constraint - dyn_parent_packC_l (PP)"!(ComponentPose=@n6, Alpha=1, ConstraintWeights_0=0.699, ConstraintWeights_1:float=0.301) -> (Pose) {
    alphaBoolBlend {blendOption=Linear}
    boneToModify {boneName=dyn_parent_packC_l}
    constraintSetup [{targetBone {boneName=pelvis}}, {targetBone {boneName=spine_01}}]
    constraintWeights [1,1]
   }
   n8 = "Constraint - dyn_parent_packA_r (PP)"!(ComponentPose=@n7, Alpha=1, ConstraintWeights_0=0.699, ConstraintWeights_1:float=0.301) -> (Pose) {
    alphaBoolBlend {blendOption=Linear}
    boneToModify {boneName=dyn_parent_packA_r}
    constraintSetup [{targetBone {boneName=pelvis}}, {targetBone {boneName=spine_01}}]
    constraintWeights [1,1]
   }
   n9 = "Constraint - dyn_parent_packB_r (PP)"!(ComponentPose=@n8, Alpha=1, ConstraintWeights_0=0.699, ConstraintWeights_1:float=0.301) -> (Pose) {
    alphaBoolBlend {blendOption=Linear}
    boneToModify {boneName=dyn_parent_packB_r}
    constraintSetup [{targetBone {boneName=pelvis}}, {targetBone {boneName=spine_01}}]
    constraintWeights [1,1]
   }
   n10 = "Constraint - dyn_parent_packC_r (PP)"!(ComponentPose=@n9, Alpha=1, ConstraintWeights_0=0.6991, ConstraintWeights_1:float=0.3009) -> (Pose) {
    alphaBoolBlend {blendOption=Linear}
    boneToModify {boneName=dyn_parent_packC_r}
    constraintSetup [{targetBone {boneName=pelvis}}, {targetBone {boneName=spine_01}}]
    constraintWeights [1,1]
   }
   n11 = "Constraint - dyn_parent_packD_l (PP)"!(ComponentPose=@n10, Alpha=1, ConstraintWeights_0=0.6991, ConstraintWeights_1:float=0.3009) -> (Pose) {
    alphaBoolBlend {blendOption=Linear}
    boneToModify {boneName=dyn_parent_packD_l}
    constraintSetup [{targetBone {boneName=pelvis}}, {targetBone {boneName=spine_01}}]
    constraintWeights [1,1]
   }
   n12 = Get DynamicsAlpha()
   n13 = Anim Dynamics - Bone: dyn_packA_01_l(ComponentPose=@n11, Alpha=@n12, ExternalForce="0,0,0") {
    angularDampingOverride=1
    angularSpringConstant=1500
    bAngularSpring=true
    bGravityOverrideInSimSpace=true
    bOverrideAngularDamping=true
    bUseGravityOverride=true
    gravityScale=0.35
    simulationSpace=RootRelative
    alphaBoolBlend {blendOption=Linear}
    alphaScaleBiasClamp {scale=0.3}
    boundBone {boneName=dyn_packA_01_l}
    componentAppliedLinearAccClamp=<3000 3000 3000>
    componentLinearAccScale=<0.3 0.3 0.3>
    componentLinearVelScale=<0.3 0.3 2>
    gravityOverride {z=-980}
    physicsBodyDefinitions [{
      sphereCollisionRadius=0
      boundBone {boneName=dyn_packA_01_l}
      boxExtents=<20 10 10>
      constraintSetup {angularLimitsMax=<5 0 5>, angularLimitsMin=<-5 -45 -5>, angularTarget=<0 0 0>}
      localJointOffset=<4 0 0>
     }]
   }
   n14 = Anim Dynamics - Bone: dyn_packB_01_l(ComponentPose=@n13, Alpha=@n12, ExternalForce="0,0,0") {
    angularDampingOverride=1
    angularSpringConstant=1500
    bAngularSpring=true
    bGravityOverrideInSimSpace=true
    bOverrideAngularDamping=true
    bUseGravityOverride=true
    gravityScale=0.35
    simulationSpace=RootRelative
    alphaBoolBlend {blendOption=Linear}
    alphaScaleBiasClamp {scale=0.5}
    boundBone {boneName=dyn_packB_01_l}
    componentAppliedLinearAccClamp=<3000 3000 3000>
    componentLinearAccScale=<0.3 0.3 0.3>
    componentLinearVelScale=<0.3 0.3 2>
    gravityOverride {z=-850}
    physicsBodyDefinitions [{
      sphereCollisionRadius=0
      boundBone {boneName=dyn_packB_01_l}
      boxExtents=<20 10 10>
      constraintSetup {angularLimitsMax=<5 0 5>, angularLimitsMin=<-5 -45 -5>, angularTarget=<0 0 0>}
      localJointOffset=<4 0 0>
     }]
   }
   n15 = Anim Dynamics - Bone: dyn_packC_01_l(ComponentPose=@n14, Alpha=@n12, ExternalForce="0,0,0") {
    angularDampingOverride=1
    angularSpringConstant=1500
    bAngularSpring=true
    bGravityOverrideInSimSpace=true
    bOverrideAngularDamping=true
    bUseGravityOverride=true
    gravityScale=0.35
    simulationSpace=RootRelative
    alphaBoolBlend {blendOption=Linear}
    alphaScaleBiasClamp {scale=0.37}
    boundBone {boneName=dyn_packC_01_l}
    componentAppliedLinearAccClamp=<3000 3000 3500>
    componentLinearAccScale=<0.3 0.3 0.3>
    componentLinearVelScale=<0.3 0.3 1.5>
    gravityOverride {z=-980}
    physicsBodyDefinitions [{
      sphereCollisionRadius=0
      boundBone {boneName=dyn_packC_01_l}
      boxExtents=<20 5 5>
      constraintSetup {angularLimitsMax=<5 0 5>, angularLimitsMin=<-5 -45 -5>, angularTarget=<0 0 0>}
      localJointOffset=<4 0 0>
     }]
   }
   n16 = Anim Dynamics - Bone: dyn_packA_01_r(ComponentPose=@n15, Alpha=@n2, ExternalForce="0,0,0") {
    angularDampingOverride=1
    angularSpringConstant=1500
    bAngularSpring=true
    bGravityOverrideInSimSpace=true
    bOverrideAngularDamping=true
    bUseGravityOverride=true
    gravityScale=0.35
    simulationSpace=RootRelative
    alphaBoolBlend {blendOption=Linear}
    alphaScaleBiasClamp {scale=0.5}
    boundBone {boneName=dyn_packA_01_r}
    componentAppliedLinearAccClamp=<3000 3000 3500>
    componentLinearAccScale=<0.3 0.3 0.3>
    componentLinearVelScale=<0.3 0.3 0.3>
    gravityOverride {z=-800}
    physicsBodyDefinitions [{
      sphereCollisionRadius=0
      boundBone {boneName=dyn_packA_01_r}
      boxExtents=<15 9 10>
      constraintSetup {angularLimitsMax=<5 0 5>, angularLimitsMin=<-5 -45 -5>, angularTarget=<0 0 0>}
      localJointOffset=<3 0 0>
     }]
   }
   n17 = Anim Dynamics - Bone: dyn_packB_01_r(ComponentPose=@n16, Alpha=@n2, ExternalForce="0,0,0") {
    angularDampingOverride=1
    angularSpringConstant=1500
    bAngularSpring=true
    bGravityOverrideInSimSpace=true
    bOverrideAngularDamping=true
    bUseGravityOverride=true
    gravityScale=0.35
    simulationSpace=RootRelative
    alphaBoolBlend {blendOption=Linear}
    alphaScaleBiasClamp {scale=0.4}
    boundBone {boneName=dyn_packB_01_r}
    componentAppliedLinearAccClamp=<3000 3000 3500>
    componentLinearAccScale=<0.3 0.3 0.3>
    componentLinearVelScale=<0.3 0.3 1.5>
    gravityOverride {z=-980}
    physicsBodyDefinitions [{
      sphereCollisionRadius=0
      boundBone {boneName=dyn_packB_01_r}
      boxExtents=<20 20 10>
      constraintSetup {angularLimitsMax=<5 0 5>, angularLimitsMin=<-5 -45 -5>, angularTarget=<0 0 0>}
      localJointOffset=<4 0 0>
     }]
   }
   n18 = Anim Dynamics - Bone: dyn_packC_01_r(ComponentPose=@n17, Alpha=@n2, ExternalForce="0,0,0") {
    angularDampingOverride=1
    angularSpringConstant=1500
    bAngularSpring=true
    bGravityOverrideInSimSpace=true
    bOverrideAngularDamping=true
    bUseGravityOverride=true
    gravityScale=0.35
    simulationSpace=RootRelative
    alphaBoolBlend {blendOption=Linear}
    alphaScaleBiasClamp {scale=0.5}
    boundBone {boneName=dyn_packC_01_r}
    componentAppliedLinearAccClamp=<3000 3000 3500>
    componentLinearAccScale=<0.3 0.3 0.3>
    componentLinearVelScale=<0.3 0.3 1.5>
    gravityOverride {z=-980}
    physicsBodyDefinitions [{
      sphereCollisionRadius=0
      boundBone {boneName=dyn_packC_01_r}
      boxExtents=<20 20 10>
      constraintSetup {angularLimitsMax=<5 0 5>, angularLimitsMin=<-5 -45 -5>, angularTarget=<0 0 0>}
      localJointOffset=<4 0 0>
     }]
   }
   n19 = Anim Dynamics - Bone: dyn_packD_01_r(ComponentPose=@n18, Alpha=@n2, ExternalForce="0,0,0") {
    angularDampingOverride=1
    angularSpringConstant=1500
    bAngularSpring=true
    bGravityOverrideInSimSpace=true
    bOverrideAngularDamping=true
    bUseGravityOverride=true
    gravityScale=0.35
    simulationSpace=RootRelative
    alphaBoolBlend {blendOption=Linear}
    alphaScaleBiasClamp {scale=0.35}
    boundBone {boneName=dyn_packD_01_r}
    componentAppliedLinearAccClamp=<2000 2000 2000>
    componentLinearAccScale=<0.3 0.3 0.3>
    componentLinearVelScale=<0.3 0.3 0.5>
    gravityOverride {z=-920}
    physicsBodyDefinitions [{
      sphereCollisionRadius=0
      boundBone {boneName=dyn_packD_01_r}
      boxExtents=<18 15 5>
      constraintSetup {angularLimitsMax=<5 0 5>, angularLimitsMin=<-5 -45 -5>, angularTarget=<0 0 0>}
      localJointOffset=<6 0 0>
     }]
   }
   n20 = "Constraint - dyn_parent_legpackA_r (PP)"!(ComponentPose=@n19, Alpha=1, ConstraintWeights_0=0.2995, ConstraintWeights_1:float=0.7005) -> (Pose) {
    alphaBoolBlend {blendOption=Linear}
    boneToModify {boneName=dyn_parent_legpackA_r}
    constraintSetup [{targetBone {boneName=thigh_r}}, {targetBone {boneName=thigh_twist_01_r}}]
    constraintWeights [1,1]
   }
   n21 = "Constraint - dyn_parent_legpackB_r (PP)"!(ComponentPose=@n20, Alpha=1, ConstraintWeights_0=0.4147, ConstraintWeights_1:float=0.5853) -> (Pose) {
    alphaBoolBlend {blendOption=Linear}
    boneToModify {boneName=dyn_parent_legpackB_r}
    constraintSetup [{targetBone {boneName=thigh_r}}, {targetBone {boneName=thigh_twist_01_r}}]
    constraintWeights [1,1]
   }
   n22 = "Transform (Modify) Bone - Bone: dyn_glute_l"(
    ComponentPose=@n21
    Alpha=Get ThighAlphaLeft() -> (ThighAlphaLeft:float)
    Translation="-3.750,-3.383,1.052"
    Rotation="0.000,0.000,-55"
    Scale="1.3,1.3,1.3"
   ) {
    rotationMode=BMM_Additive
    rotationSpace=BCS_BoneSpace
    scaleMode=BMM_Additive
    scaleSpace=BCS_BoneSpace
    translationMode=BMM_Additive
    translationSpace=BCS_BoneSpace
    alphaBoolBlend {blendOption=Linear}
    alphaScaleBiasClamp {bInterpResult=true, interpSpeedDecreasing=15, interpSpeedIncreasing=15}
    boneToModify {boneName=dyn_glute_l}
    rotation {roll=-55}
    scale=<1.3 1.3 1.3>
    translation=<-3.75 -3.383 1.052>
   }
   n23 = "Transform (Modify) Bone - Bone: dyn_glute_r"(
    ComponentPose=@n22
    Alpha=Get ThighAlphaRight() -> (ThighAlphaRight:float)
    Translation="3.750,0,0"
    Rotation="0.000,0.000,-40"
    Scale="1.25,1.25,1.25"
   ) {
    rotationMode=BMM_Additive
    rotationSpace=BCS_BoneSpace
    scaleMode=BMM_Additive
    scaleSpace=BCS_BoneSpace
    translationMode=BMM_Additive
    translationSpace=BCS_BoneSpace
    alphaBoolBlend {blendOption=Linear}
    boneToModify {boneName=dyn_glute_r}
    rotation {roll=-40}
    scale=<1.25 1.25 1.25>
    translation {x=3.75}
   }
   n24 = Get CrouchAlpha() -> (CrouchAlpha:double)
   n25 = "Transform (Modify) Bone - Bone: dyn_driver_packA_l"(ComponentPose=@n23, Alpha=@n24, Translation="-0.798,-7.566,-1.628", Rotation="0.000,-0.000,-57.221", Scale="1,1,1") {
    rotationMode=BMM_Additive
    rotationSpace=BCS_BoneSpace
    translationMode=BMM_Additive
    translationSpace=BCS_BoneSpace
    alphaBoolBlend {blendOption=Linear}
    boneToModify {boneName=dyn_driver_packA_l}
    rotation {roll=-57.221}
    translation=<-0.7979 -7.566 -1.628>
   }
   n26 = "Transform (Modify) Bone - Bone: dyn_driver_packB_l"(ComponentPose=@n25, Alpha=@n24, Translation="0.000,-3.730,-0.000", Rotation="0.000,0.000,-18.183", Scale="1,1,1") {
    rotationMode=BMM_Additive
    rotationSpace=BCS_BoneSpace
    translationMode=BMM_Additive
    translationSpace=BCS_BoneSpace
    alphaBoolBlend {blendOption=Linear}
    boneToModify {boneName=dyn_driver_packB_l}
    rotation {roll=-18.183}
    translation {y=-3.73}
   }
   n27 = "Transform (Modify) Bone - Bone: dyn_gun_holster_l"(ComponentPose=@n26, Alpha=@n24, Translation="0.063,0.324,-0.000", Rotation="0.000,-23,0.000", Scale="1,1,1") {
    rotationMode=BMM_Additive
    rotationSpace=BCS_BoneSpace
    translationMode=BMM_Additive
    translationSpace=BCS_BoneSpace
    alphaBoolBlend {blendOption=Linear}
    boneToModify {boneName=dyn_gun_holster_l}
    rotation {yaw=-23}
    translation=<0.063 0.324>
   }
   Output Pose(Result=Component To Local(ComponentPose=@n27)) {layerGroup=None}
   Reroute Node(InputPin:Wildcard) -> (OutputPin:Wildcard)
  ]}
 EventGraph {nodes [
   Event OnBeginCrouching()
   > Set CrouchAlpha(CrouchAlpha:double=1) -> (Output_Get:double)
   Event OnEndCrouching()
   > Set CrouchAlpha(CrouchAlpha=0)
  ]}
]
variables [CrouchAlpha {type=number}]
